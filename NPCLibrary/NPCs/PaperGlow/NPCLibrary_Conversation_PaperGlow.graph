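# Conversation database for the NPC PaperGlow, serving as its main greeting and branching hub.
# The entry point leads to one speech node with contextual lines for first-ever meetings, specific player outfits, and per-match conversation-count checks, plus a default myth-themed line.
# Priority curve rows weight the first-meeting and outfit reactions.
# The speech then links by tag to shared NPC quests, basic services, and PaperGlow's own dialogue entry.
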ConversationDatabase /NPCLibrary/NPCs/PaperGlow/NPCLibrary_Conversation_PaperGlow
entry_points [{tag=Conversation.Entry.PaperGlow, destinations [#1]}]
nodes [
 ConversationEntryPointNode_7 {
  id=#1
  class=ConversationEntryPointNode
  outputs [#2]
  properties {entryTag {tagName=Conversation.Entry.PaperGlow}}
 }
 Speech_C_4 {
  id=#2
  class=Speech_C
  outputs [#3, #4, #5]
  properties {
   nodeName=""
   generalConfig {
    defaultMessage="Every myth, every story, carries a kernel of truth at its heart."
    contextualMessages [
     {
      message="May the moon walk with you, stranger."
      contextRequirements [{requirement {class=FortControllerRequirement_IsFirstEverConversationWithNPC}}]
      priority {curve {rowName=Tandem.Speech.Priority.FirstEverInteraction}}
     }
     {
      message="I have heard the myths, Medusa. But I'd love to hear your story from you."
      contextRequirements [{requirement {
         class=FortControllerRequirement_HasCID
         properties {allowedCIds [/BRCosmetics/Athena/Items/Cosmetics/Characters/Character_ReptilianOcean_Sleek]}
        }}]
      priority {curve {rowName=Tandem.Speech.Priority.OutfitReaction}}
     }
     {
      message=Many stories end with hope. But your story is just beginning.
      contextRequirements [{requirement {
         class=FortControllerRequirement_HasCID
         properties {allowedCIds [&Character_VoidRedemption_Rebel]}
        }}]
      priority {curve {rowName=Tandem.Speech.Priority.OutfitReaction}}
     }
     {
      message="What mysterious, wonderful stories I see in your eyes, stranger."
      contextRequirements [{requirement {
         class=FortControllerRequirement_HasCID
         properties {allowedCIds [&Character_UndergroundRebel_Fashion]}
        }}]
      priority {curve {rowName=Tandem.Speech.Priority.OutfitReaction}}
     }
     {
      message=You say mortality is no threat to you... but your actions betray your fear.
      contextRequirements [{requirement {
         class=FortControllerRequirement_HasCID
         properties {allowedCIds [&Character_StunningMountain_Armor]}
        }}]
      priority {curve {rowName=Tandem.Speech.Priority.OutfitReaction}}
     }
     {
      message=Stories are more resilient than you give them credit for.
      contextRequirements [{requirement {class=FortControllerRequirement_HasCID, properties {allowedCIds [/BRCosmetics/Athena/Items/Cosmetics/Characters/Character_PaperBlaze]}}}]
      priority {curve {rowName=Tandem.Speech.Priority.OutfitReaction}}
     }
     {
      message="Every myth, every story, carries a kernel of truth at its heart."
      contextRequirements [{requirement {class=FortControllerRequirement_CheckNPCMatchConversationCount, properties {nPCUniqueIdTag {tagName=AISpawnerData.Type.Tandem.PaperGlow}}}}]
      priority {curve {rowName=None}}
     }
     {
      message=I write down the stories so they are not forgotten by those to come.
      contextRequirements [{requirement {class=FortControllerRequirement_CheckNPCMatchConversationCount, properties {nPCUniqueIdTag {tagName=AISpawnerData.Type.Tandem.PaperGlow}}}}]
      priority {curve {rowName=None}}
     }
     {
      message=Mythology is merely another form of memory.
      contextRequirements [{requirement {class=FortControllerRequirement_CheckNPCMatchConversationCount, properties {nPCUniqueIdTag {tagName=AISpawnerData.Type.Tandem.PaperGlow}}}}]
      priority {curve {rowName=None}}
     }
     {
      message="Stories contain power. They tell us who we are, and help us shape the world."
      contextRequirements [{requirement {class=FortControllerRequirement_CheckNPCMatchConversationCount, properties {nPCUniqueIdTag {tagName=AISpawnerData.Type.Tandem.PaperGlow}}}}]
      priority {curve {rowName=None}}
     }
    ]
   }
  }
 }
 ConversationLinkNode_2 {id=#5, class=ConversationLinkNode, properties {remoteEntryTag {tagName=Conversation.Entry.PaperGlow.Dialogue}}}
 ConversationLinkNode_0 {id=#4, class=ConversationLinkNode, properties {remoteEntryTag {tagName=Conversation.Entry.BasicServices}}}
 ConversationLinkNode_1 {id=#3, class=ConversationLinkNode, properties {remoteEntryTag {tagName=Conversation.Entry.NPCLibrary.Quests}}}
]
